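# UI-domain translucent material, named as the shell of the top bar in the UI refresh set.
# The graph holds only two parameters: a Color vector defaulting to dark navy and an Alpha scalar defaulting to fully opaque.
# It has no textures or math nodes, and no links from the parameters to material outputs are shown.
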Material /Game/UI/Foundation/UI_Refresh/M_UI_TopBar_Shell
graph {
 properties {
  bUsedWithStaticMesh=false
  blendMode=BLEND_Translucent
  materialDomain=MD_UI
  pixelDepthOffsetMode=PDOM_Legacy
  usedShadingModels=Default Lit
 }
 nodes [
  ScalarParameter() {defaultValue=1, parameterName=Alpha}
  VectorParameter() {parameterName=Color, defaultValue=<rgba 0.0019 0.0183 0.125 1>}
 ]
}
